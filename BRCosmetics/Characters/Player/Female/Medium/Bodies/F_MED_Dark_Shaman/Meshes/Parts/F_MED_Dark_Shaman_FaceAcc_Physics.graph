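# Physics asset for the female medium Dark Shaman face accessory mesh, providing bodies and joints for dangling dreadlocks and an earring.
# Dreadlock chains of one to four bones, mostly capsules with boxes on chain J, hang from the head under tight swing and twist limits.
# The box earring also constrains to the head.
# Capsules on upper spine, neck, head and right clavicle form the body, linked by free-rotation constraints.
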PhysicsAsset /BRCosmetics/Characters/Player/Female/Medium/Bodies/F_MED_Dark_Shaman/Meshes/Parts/F_MED_Dark_Shaman_FaceAcc_Physics
bodies {
 clavicle_r [{shapeType=Capsule, length=15.125, radius=5.255, center=<-6.9612 2.2567 0.0001>, rotation=<rot 90 0 0>}]
 dyn_DredA [{shapeType=Capsule, length=4.1423, radius=0.8227, center=<2.5191 0.1776 -0.2503>, rotation=<rot -0.6298 95.6471 -94.9638>}]
 dyn_DredB_01 [{shapeType=Capsule, length=4.5659, radius=0.8352, center=<2.7206 0.0001 0>, rotation=<rot 0 90 90>}]
 dyn_DredB_02 [{shapeType=Capsule, length=4.8808, radius=0.8461, center=<2.7159 0.0001 0>, rotation=<rot 0 90 90>}]
 dyn_DredC_01 [{shapeType=Capsule, length=3.517, radius=0.6144, center=<1.883 0.0497 0.1049>, rotation=<rot 0 -89.2228 -88.8017>}]
 dyn_DredC_02 [{shapeType=Capsule, length=3.6321, radius=0.5792, center=<2.2103 -0.0439 0.0686>, rotation=<rot 0 -88.398 -89.1788>}]
 dyn_DredD_01 [{shapeType=Capsule, length=2.8675, radius=0.5615, center=<1.4297 0 0>, rotation=<rot 90 0 0>}]
 dyn_DredD_02 [{shapeType=Capsule, length=3.5687, radius=0.5891, center=<2.1249 0 0>, rotation=<rot 90 0 0>}]
 dyn_DredE_01 [{shapeType=Capsule, length=4.0279, radius=0.8931, center=<2.4199 0 0.0001>, rotation=<rot 90 0 0>}]
 dyn_DredE_02 [{shapeType=Capsule, length=5.7135, radius=0.6594, center=<3.0092 0 -0.0001>, rotation=<rot 90 0 0>}]
 dyn_DredE_03 [{shapeType=Capsule, length=5.0873, radius=0.5, center=<2.789 0 0>, rotation=<rot 90 0 0>}]
 dyn_DredF_01 [{shapeType=Capsule, length=3.9451, radius=1.0032, center=<2.5434 0.0001 0>, rotation=<rot 90 0 0>}]
 dyn_DredF_02 [{shapeType=Capsule, length=4.7942, radius=0.6223, center=<2.8583 0 0>, rotation=<rot 90 0 0>}]
 dyn_DredF_03 [{shapeType=Capsule, length=2.8368, radius=0.505, center=<1.5651 0 0>, rotation=<rot 90 0 0>}]
 dyn_DredG_01 [{shapeType=Capsule, length=4.6856, radius=1.1221, center=<3.0571 -0.1594 0.0378>, rotation=<rot 0 -87.9524 -89.6365>}]
 dyn_DredG_02 [{shapeType=Capsule, length=5.8525, radius=0.7553, center=<3.1424 0.1763 -0.0668>, rotation=<rot 0 87.9024 -87.9733>}]
 dyn_DredG_03 [{shapeType=Capsule, length=4.6259, radius=0.5, center=<2.5476 0.1532 0.0694>, rotation=<rot 0 -91.6189 -88.8894>}]
 dyn_DredG_04 [{shapeType=Capsule, length=2.7573, radius=0.5392, center=<1.7073 0.0731 0.0841>, rotation=<rot 0 -90.2943 -89.9813>}]
 dyn_DredH_01 [{shapeType=Capsule, length=3.1165, radius=0.968, center=<1.8877 -0.0305 -0.0774>, rotation=<rot 0 -91.6473 -88.159>}]
 dyn_DredH_02 [{shapeType=Capsule, length=3.6026, radius=0.724, center=<2.1701 -0.1246 0.1121>, rotation=<rot 0 -87.81 -87.4762>}]
 dyn_DredJ_01 [{shapeType=Box, extentX=2.6498, extentY=6.7826, extentZ=5.357, center=<3.3549 0.0263 -0.2613>, rotation=<rot -7.0342 -89.9999 -6.5673>}]
 dyn_DredJ_02 [{shapeType=Box, extentX=2.2201, extentY=4.9726, extentZ=9.9523, center=<4.8224 -0.3932 0.0754>, rotation=<rot 0.0844 -93.1325 -88.4458>}]
 dyn_DredK01 [{shapeType=Capsule, length=4.4066, radius=1.0256, center=<3.1014 -0.3088 -0.1524>, rotation=<rot 0 94.3297 -87.7998>}]
 dyn_DredK_02 [{shapeType=Capsule, length=3.5537, radius=0.9421, center=<2.3036 -0.1199 -0.1162>, rotation=<rot 0 90.3737 -87.2967>}]
 dyn_DredK_03 [{shapeType=Capsule, length=4.4223, radius=0.726, center=<2.5134 0.1217 -0.1048>, rotation=<rot 0 -92.2187 -87.7523>}]
 dyn_DredL01 [{shapeType=Capsule, length=2.9708, radius=1.0427, center=<1.5233 -0.1035 0.0592>, rotation=<rot 6.0857 -86.4876 -88.247>}]
 dyn_DredL02 [{shapeType=Capsule, length=4.6476, radius=1.0612, center=<3.2469 -0.1269 0.1095>, rotation=<rot 0 -87.9638 -87.8697>}]
 dyn_DredL03 [{shapeType=Capsule, length=7.1652, radius=0.732, center=<3.9 -0.0228 -0.2121>, rotation=<rot 0.1849 82.0407 -88.9263>}]
 dyn_DredM01 [{shapeType=Capsule, length=3.7654, radius=1.0848, center=<2.4231 -0.1528 0.0895>, rotation=<rot 0 91.7977 -87.4666>}]
 dyn_DredM02 [{shapeType=Capsule, length=3.4993, radius=0.8215, center=<2.2839 -0.1263 0.1308>, rotation=<rot 0 -87.0095 -88.8256>}]
 dyn_DredM03 [{shapeType=Capsule, length=4.4734, radius=0.6437, center=<2.1761 0.0711 -0.0228>, rotation=<rot 0 83.128 -88.734>}]
 dyn_earring [{shapeType=Box, extentX=1.1449, extentY=4.085, extentZ=4.6219, center=<1.7225 0.0024 0.0077>, rotation=<rot 0 90.0765 -89.8269>}]
 head [{shapeType=Capsule, length=7.8732, radius=6.4512, center=<2.0466 -2.3523 -0.493>, rotation=<rot 77.41 -73.0581 -23.4561>}]
 neck_01 [{shapeType=Capsule, length=5.25, radius=4.505, rotation=<rot 90 0 0>}]
 neck_02 [{shapeType=Capsule, length=5.25, radius=4.505, rotation=<rot 90 0 0>}]
 spine_05 [{shapeType=Capsule, length=19.5524, radius=10.4476, center=<-4.2911 0 0>}]
}
constraints [
 {bone1Name=dyn_DredM01, bone2Name=head, swing1LimitDegrees=8, swing2LimitDegrees=8, twistLimitDegrees=8}
 {bone1Name=dyn_DredM02, bone2Name=dyn_DredM01, swing1LimitDegrees=8, swing2LimitDegrees=8, twistLimitDegrees=8}
 {bone1Name=dyn_DredM03, bone2Name=dyn_DredM02, swing1LimitDegrees=8, swing2LimitDegrees=8, twistLimitDegrees=8}
 {bone1Name=dyn_DredL01, bone2Name=head, swing1LimitDegrees=8, swing2LimitDegrees=8, twistLimitDegrees=8}
 {bone1Name=dyn_DredL02, bone2Name=dyn_DredL01, swing1LimitDegrees=8, swing2LimitDegrees=8, twistLimitDegrees=8}
 {bone1Name=dyn_DredL03, bone2Name=dyn_DredL02, swing1LimitDegrees=8, swing2LimitDegrees=8, twistLimitDegrees=8}
 {bone1Name=dyn_DredK01, bone2Name=head, swing1LimitDegrees=15, swing2LimitDegrees=10, twistLimitDegrees=5}
 {bone1Name=dyn_DredK_02, bone2Name=dyn_DredK01, swing1LimitDegrees=15, swing2LimitDegrees=10, twistLimitDegrees=5}
 {bone1Name=dyn_DredK_03, bone2Name=dyn_DredK_02, swing1LimitDegrees=15, swing2LimitDegrees=10, twistLimitDegrees=5}
 {bone1Name=dyn_DredJ_01, bone2Name=head, swing1LimitDegrees=15, swing2LimitDegrees=10, twistLimitDegrees=5}
 {bone1Name=dyn_DredJ_02, bone2Name=dyn_DredJ_01, swing1LimitDegrees=15, swing2LimitDegrees=10, twistLimitDegrees=5}
 {bone1Name=dyn_DredH_01, bone2Name=head, swing1LimitDegrees=10, swing2LimitDegrees=10, twistLimitDegrees=8}
 {bone1Name=dyn_DredH_02, bone2Name=dyn_DredH_01, swing1LimitDegrees=10, swing2LimitDegrees=10, twistLimitDegrees=8}
 {bone1Name=dyn_DredG_01, bone2Name=head, swing1LimitDegrees=10, swing2LimitDegrees=10, twistLimitDegrees=8}
 {bone1Name=dyn_DredG_02, bone2Name=dyn_DredG_01, swing1LimitDegrees=10, swing2LimitDegrees=10, twistLimitDegrees=8}
 {bone1Name=dyn_DredG_03, bone2Name=dyn_DredG_02, swing1LimitDegrees=10, swing2LimitDegrees=10, twistLimitDegrees=8}
 {bone1Name=dyn_DredG_04, bone2Name=dyn_DredG_03, swing1LimitDegrees=10, swing2LimitDegrees=10, twistLimitDegrees=8}
 {bone1Name=dyn_DredF_01, bone2Name=head, swing1LimitDegrees=10, swing2LimitDegrees=10, twistLimitDegrees=8}
 {bone1Name=dyn_DredF_02, bone2Name=dyn_DredF_01, swing1LimitDegrees=10, swing2LimitDegrees=10, twistLimitDegrees=8}
 {bone1Name=dyn_DredF_03, bone2Name=dyn_DredF_02, swing1LimitDegrees=10, swing2LimitDegrees=10, twistLimitDegrees=8}
 {bone1Name=dyn_DredE_01, bone2Name=head, swing1LimitDegrees=10, swing2LimitDegrees=10, twistLimitDegrees=8}
 {bone1Name=dyn_DredE_02, bone2Name=dyn_DredE_01, swing1LimitDegrees=10, swing2LimitDegrees=10, twistLimitDegrees=8}
 {bone1Name=dyn_DredE_03, bone2Name=dyn_DredE_02, swing1LimitDegrees=10, swing2LimitDegrees=10, twistLimitDegrees=8}
 {bone1Name=dyn_DredD_01, bone2Name=head, swing1LimitDegrees=10, swing2LimitDegrees=10, twistLimitDegrees=8}
 {bone1Name=dyn_DredD_02, bone2Name=dyn_DredD_01, swing1LimitDegrees=10, swing2LimitDegrees=10, twistLimitDegrees=8}
 {bone1Name=dyn_DredC_01, bone2Name=head, swing1LimitDegrees=10, swing2LimitDegrees=10, twistLimitDegrees=8}
 {bone1Name=dyn_DredC_02, bone2Name=dyn_DredC_01, swing1LimitDegrees=10, swing2LimitDegrees=10, twistLimitDegrees=8}
 {bone1Name=dyn_DredB_01, bone2Name=head, swing1LimitDegrees=10, swing2LimitDegrees=10, twistLimitDegrees=8}
 {bone1Name=dyn_DredB_02, bone2Name=dyn_DredB_01, swing1LimitDegrees=10, swing2LimitDegrees=10, twistLimitDegrees=8}
 {bone1Name=dyn_DredA, bone2Name=head, swing1LimitDegrees=10, swing2LimitDegrees=10, twistLimitDegrees=8}
 {bone1Name=dyn_earring, bone2Name=head, swing1LimitDegrees=15, swing2LimitDegrees=10, twistLimitDegrees=5}
 {
  bone1Name=clavicle_r
  bone2Name=spine_05
  swing1LimitDegrees=45
  swing1Motion=Free
  swing2LimitDegrees=45
  swing2Motion=Free
  twistLimitDegrees=45
  twistMotion=Free
 }
 {
  bone1Name=head
  bone2Name=neck_02
  swing1LimitDegrees=45
  swing1Motion=Free
  swing2LimitDegrees=45
  swing2Motion=Free
  twistLimitDegrees=45
  twistMotion=Free
 }
 {
  bone1Name=neck_01
  bone2Name=spine_05
  swing1LimitDegrees=45
  swing1Motion=Free
  swing2LimitDegrees=45
  swing2Motion=Free
  twistLimitDegrees=45
  twistMotion=Free
 }
 {
  bone1Name=neck_02
  bone2Name=neck_01
  swing1LimitDegrees=45
  swing1Motion=Free
  swing2LimitDegrees=45
  swing2Motion=Free
  twistLimitDegrees=45
  twistMotion=Free
 }
]
